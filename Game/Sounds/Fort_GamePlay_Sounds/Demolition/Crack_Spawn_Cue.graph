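# Sound cue in the demolition gameplay set that plays when a crack spawns, routed through the melee weapon sound class.
# A mixer layers a randomly chosen crack-spawn variation, at lower volume, over a fixed crack test wave.
# Each layer has its own random upward pitch modulation, compounded by a raised cue-wide pitch.
# Crack-specific attenuation and the player melee concurrency set are assigned, with priming on load and the debug flag left on.
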SoundCue /Game/Sounds/Fort_GamePlay_Sounds/Demolition/Crack_Spawn_Cue
root=Mixer_1
nodes [
 {id=Mixer_1, children [Modulator_0,Modulator_1], properties {inputVolume [0.6,1]}}
 {id=Modulator_0, children [Random_0], properties {pitchMax=1.6, pitchMin=1.4}}
 {id=Modulator_1, children [WavePlayer_34], properties {pitchMax=1.5, pitchMin=1}}
 {id=Random_0, children [WavePlayer_0,WavePlayer_1,WavePlayer_2,WavePlayer_3], properties {weights [1,1,1,1]}}
 {id=WavePlayer_0, properties {soundWaveAssetPtr=&/Crack_Spawn_01}}
 {id=WavePlayer_1, properties {soundWaveAssetPtr=&/Crack_Spawn_02}}
 {id=WavePlayer_2, properties {soundWaveAssetPtr=&/Crack_Spawn_03}}
 {id=WavePlayer_3, properties {soundWaveAssetPtr=&/Crack_Spawn_04}}
 {id=WavePlayer_34, properties {soundWaveAssetPtr=&/crack_test_02}}
]
properties {
 attenuationSettings=/Game/Sounds/Attenuation/Fort_SoundCrack_Att
 bDebug=true
 bPrimeOnLoad=true
 maxDistance=1300
 pitchMultiplier=1.5
 soundClassObject=&Fort_WeaponMelee_PSM
 volumeMultiplier=0.6
 concurrencyOverrides {maxCount=4, resolutionRule=StopFarthestThenPreventNew}
 concurrencySet [&Player_Melee_Weapon_Conc]
}
